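# Boss health bar HUD widget for the Mash LTM, showing the boss's name above a sheared fill bar.
# Boss fight start and health update events set the health text and drive the bar's dynamic material fill from the health ratio.
# UpdateBossDisplayName sets the name from the Mash mutator, retrying on a short timer until a player-pawn boss has a valid PlayerState.
# The health text starts collapsed, and its show node has no incoming trigger.
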WidgetBlueprint /Game/Athena/HUD/LTM/Mash/HUD_Mash_BossBattle
graphs [EventGraph {nodes [
   n1 = SetVisibility(self=Get BossHealth_Text() -> (BossHealth_Text:Text Object Reference), InVisibility=HitTestInvisible)
   n2 = Get BossHealthRatio() -> (BossHealthRatio:float)
   "SetText (Text)"(
    execute=[@n1.then, @n11.then, @n13.then]
    self=Get BossHealth_Text()
    InText=Format Text!(
     Format="LOCTEXT(\"Super Fiend - {CUR_HEALTH} / {TOTAL_HEALTH} ({RATIO})\")"
     CUR_HEALTH:float=Get CurrentBossHealth() -> (CurrentBossHealth:float)
     TOTAL_HEALTH:float=Get TotalBossHealth() -> (TotalBossHealth:float)
     RATIO:float=@n2
    ) -> (Result)
   )
   > SetScalarParameterValue(
    self=Get HealthBarMID() -> (HealthBarMID:Material Instance Dynamic Object Reference)
    ParameterName=Get MaterialParamHealthFill Const() -> (MaterialParamHealthFill Const:Name=None)
    Value=@n2
   )
   UpdateBossDisplayName() {is=title K2Node_CustomEvent}
   > n3 = Get!(execute:Exec) -> (then:Exec, else:Exec, MashMutatorRef:Fort Athena Mutator Mash Object Reference)
   n4 = Get BossPawn!(self:Fort Athena Mutator Mash Object Reference=@n3) -> (BossPawn:Fort Pawn Object Reference)
   > n5 = Cast To FortPlayerPawnAthena(Object=@n4)
   n6 = Is Valid(exec=[@n8.then, @n5.then], InputObject=Get PlayerState!(self:Pawn Object Reference=@n5) -> (PlayerState:Player State Object Reference))
   n7 = Is Valid(exec=@n6.Is Not Valid, InputObject=@n4)
   Clear and Invalidate Timer by Handle(execute=[@n10.then, @n7.Is Not Valid], Handle=Get checkBossNameTimer() -> (checkBossNameTimer:Timer Handle Structure))
   n8 = WaitForPlayerState() {is=title K2Node_CustomEvent}
   n9 = Set Timer by Event(execute=@n7.Is Valid, Delegate=@n8, Time=0.25, bLooping=false)
   > Set checkBossNameTimer(checkBossNameTimer:Timer Handle Structure=@n9) -> (Output_Get:Timer Handle Structure)
   > "SetText (Text)"(
    self=Get DisplayName() -> (DisplayName:Common Text Object Reference)
    InText=GetBossDisplayName!(self:Fort Athena Mutator Mash Object Reference=@n3) -> (ReturnValue:Text)
   )
   n10 = "SetText (Text)"(execute=[@n6.Is Valid, @n5.CastFailed], self=Get DisplayName(), InText=GetBossDisplayName(self=@n3))
   n11 = Event OnBossFightStarted() {is=title K2Node_Event}
   Event Construct()
   > n12 = GetDynamicMaterial(self=Get Image_HealthBar() -> (Image_HealthBar:Image Object Reference))
   > Set HealthBarMID(HealthBarMID:Material Instance Dynamic Object Reference=@n12) -> (Output_Get:Material Instance Dynamic Object Reference)
   n13 = Event OnBossHealthUpdate() {is=title K2Node_Event}
  ]}]
variables [
 checkBossNameTimer {type=TimerHandle}
 HealthBarMID {type=MaterialInstanceDynamic}
 MaterialParamHealthFill Const {type=string, default=PercentFill}
]
widgets {
 info {inheritedWidgetCount=1, parentClass=/Script/FortniteUI.MashBossBattleWidget, rootWidgetClass=&Border, widgetCount=10}
 widgets [
  BorderContainer {
   widgetClassPath=&Border
   widget=BorderContainer
   properties {
    visibility=HitTestInvisible
    background {
     bIsDynamicallyLoaded=false
     drawAs=NoDrawType
     imageType=NoImage
     mirroring=NoMirror
     resourceName=None
     resourceObject=null
     tiling=NoTile
     imageSize=<32 32>
     margin=<ltrb 0 0 0 0>
     outlineSettings {
      bUseBrushTransparency=false
      roundingType=HalfHeightRadius
      width=0
      color {colorUseRule=UseColor_Specified, specifiedColor=<rgba 0 0 0 0>}
      cornerRadii=<0 0 0 0>
     }
     tintColor {colorUseRule=UseColor_Specified, specifiedColor=<rgba 1 1 1 1>}
     uVRegion {bIsValid=false, max=<0 0>, min=<0 0>}
    }
    padding=<ltrb 0 0 0 0>
   }
  }
  Overlay_0 {widgetClassPath=&Overlay, parent=BorderContainer, widget=Overlay_0, slot {class=BorderSlot, properties {padding=<ltrb 0 0 0 0>}}}
  VerticalBox_0 {
   widgetClassPath=&VerticalBox
   parent=Overlay_0
   widget=VerticalBox_0
   slot {class=OverlaySlot, properties {horizontalAlignment=HAlign_Center}}
  }
  DisplayName {
   widgetClassPath=&CommonTextBlock
   bIsVariable=true
   parent=VerticalBox_0
   widget=DisplayName
   properties {
    text=Final Boss
    textTransformPolicy=ToUpper
    font {
     bForceMonospaced=false
     bMaterialIsStencil=false
     fontMaterial=null
     fontObject=&Burbank
     letterSpacing=0
     monospacedWidth=1
     size=24
     skewAmount=0
     typefaceFontName=BurbankRegular-Black
     outlineSettings {
      bApplyOutlineToDropShadows=false
      bMiteredCorners=false
      bSeparateFillAlpha=false
      outlineMaterial=null
      outlineSize=1
      outlineColor=<rgba 0.0194 0 0.0423 0.5>
     }
    }
    shadowColorAndOpacity=<rgba 0.0529 0.0091 0.3005 0>
    shadowOffset=<3 3>
   }
   slot {class=VerticalBoxSlot, properties {horizontalAlignment=HAlign_Center, verticalAlignment=VAlign_Top}}
  }
  SizeBox_0 {
   widgetClassPath=&SizeBox
   parent=VerticalBox_0
   widget=SizeBox_0
   properties {bOverride_HeightOverride=true, bOverride_WidthOverride=true, heightOverride=24.9221, widthOverride=617.9141}
   slot {class=VerticalBoxSlot, properties {horizontalAlignment=HAlign_Center, verticalAlignment=VAlign_Center}}
  }
  Overlay_1 {
   widgetClassPath=&Overlay
   parent=SizeBox_0
   widget=Overlay_1
   properties {renderTransform {angle=0, scale=<1 1>, shear=<-12 0>, translation=<0 0>}}
   slot {class=SizeBoxSlot}
  }
  Image_Shadow {
   widgetClassPath=&Image
   parent=Overlay_1
   widget=Image_Shadow
   properties {colorAndOpacity=<rgba 0.0052 0.0052 0.0056 0.4>, renderTransform {angle=0, scale=<1 1>, shear=<0 0>, translation=<4 4>}}
   slot {class=OverlaySlot, properties {horizontalAlignment=HAlign_Fill, verticalAlignment=VAlign_Fill}}
  }
  Image_HealthBar {
   widgetClassPath=&Image
   bIsVariable=true
   parent=Overlay_1
   widget=Image_HealthBar
   properties {brush {
     bIsDynamicallyLoaded=false
     drawAs=Box
     imageType=NoImage
     mirroring=NoMirror
     resourceName=None
     resourceObject=/Game/Athena/HUD/LTM/Mash/Materials/M_UI_BossHealth
     tiling=NoTile
     imageSize=<32 32>
     margin=<ltrb 0.2 0.2 0.2 0.2>
     outlineSettings {
      bUseBrushTransparency=false
      roundingType=HalfHeightRadius
      width=0
      color {colorUseRule=UseColor_Specified, specifiedColor=<rgba 0 0 0 0>}
      cornerRadii=<0 0 0 0>
     }
     tintColor {colorUseRule=UseColor_Specified, specifiedColor=<rgba 1 1 1 1>}
     uVRegion {bIsValid=false, max=<0 0>, min=<0 0>}
    }}
   slot {class=OverlaySlot, properties {horizontalAlignment=HAlign_Fill, verticalAlignment=VAlign_Fill}}
  }
  BossHealth_Text {
   widgetClassPath=&TextBlock
   bIsVariable=true
   parent=Overlay_1
   widget=BossHealth_Text
   properties {
    justification=Center
    text=120/500
    visibility=Collapsed
    font {
     bForceMonospaced=false
     bMaterialIsStencil=false
     fontMaterial=null
     fontObject=&BurbankSmall
     letterSpacing=0
     monospacedWidth=1
     size=15
     skewAmount=0
     typefaceFontName=Bold
     outlineSettings {
      bApplyOutlineToDropShadows=false
      bMiteredCorners=false
      bSeparateFillAlpha=false
      outlineMaterial=null
      outlineSize=1
      outlineColor=<rgba 0.1878 0 0.6445 1>
     }
    }
   }
   slot {class=OverlaySlot, properties {horizontalAlignment=HAlign_Center, verticalAlignment=VAlign_Center}}
  }
  TutorialHighlight {widgetClassPath=&TutorialHighlight}
 ]
}
