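# Translucent UI material, named for a focus border in login sub-game select, that draws a repeating pulse of an animated, fading border.
# Time plus an offset, scaled by speed, wraps through Fmod, is optionally reversed by a static switch, then shaped by a power curve.
# That phase scales two center-scaled 9-slice UV masks, whose outer-times-inverted-inner product forms the band.
# The same phase fades the band from full to zero.
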Material /Game/UI/Login/SubGameSelect/Materials/M_UI_FocusBorderPulse
graph {
 properties {
  bUsedWithStaticMesh=false
  blendMode=BLEND_Translucent
  materialDomain=MD_UI
  pixelDepthOffsetMode=PDOM_Legacy
  usedShadingModels=Default Lit
 }
 nodes [
  Constant() {r=1}
  n1 = MaterialFunctionCall() {materialFunction=&GetUserInterfaceUV}
  n2 = Time()
  n3 = ScalarParameter() {parameterName=TimeOffset}
  n4 = Fmod(A=Multiply(A=Add(A=@n2, B=@n3), B=ScalarParameter() {defaultValue=1, parameterName=PulseSpeed}), B=Constant() {r=1})
  n5 = Power(
   Base=StaticSwitchParameter(True=OneMinus(None=@n4), False=@n4) {parameterName=Reverse}
   Exp=ScalarParameter() {defaultValue=2, parameterName=FalloffPower}
  )
  n6 = LinearInterpolate(
   A=ScalarParameter() {defaultValue=0.9, parameterName=OuterStartScale}
   B=ScalarParameter() {defaultValue=1, parameterName=OuterEndScale}
   Alpha=@n5
  )
  n7 = MaterialFunctionCall(
   UVs=@n1.9-Slice UV
   Texture Scale=LinearInterpolate(A=ScalarParameter() {defaultValue=0.8, parameterName=StartScale}, B=ScalarParameter() {defaultValue=1, parameterName=EndScale}, Alpha=@n5)
  ) {materialFunction=&ScaleUVsByCenter}
  Multiply(
   A=Multiply(
    A=MaterialFunctionCall(UVs=@n1.9-Slice UV, Texture Scale=@n6) {materialFunction=&ScaleUVsByCenter}.0-1 mask
    B=OneMinus(None=@n7.0-1 mask)
   )
   B=LinearInterpolate(Alpha=@n5) {constA=1, constB=0}
  )
 ]
}
