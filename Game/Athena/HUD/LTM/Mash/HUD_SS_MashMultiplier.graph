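# HUD widget for a Mash LTM multiplier bonus indicator, with a compact icon view and a detailed label view.
# EventShowDetail sets a widget switcher from a bool: false shows a pulsing border over a world-indicator image, true shows "Multiplier Bonus" text with an arrow.
# EventIntro plays IntroScale, which shrinks the icon overlay from oversized to resting scale while its opacity fades in, flickers and finally fades to zero.
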WidgetBlueprint /Game/Athena/HUD/LTM/Mash/HUD_SS_MashMultiplier
animations [IntroScale {
  binding_count=1
  length_seconds=3.7327
  object=&self:IntroScale
  sequence {bindings [Overlay_0 {
     id=#1
     possessed_class=Overlay
     tracks [
      {
       type=MovieScene2DTransformTrack
       sections [{
         type=MovieScene2DTransformSection
         channels [
          Scale.X {type=MovieSceneScriptingFloatChannel, keys [[0,3.9789],[29.8347,0.65]]}
          Scale.Y {type=MovieSceneScriptingFloatChannel, keys [[0,4.0789],[29.8347,0.75]]}
         ]
        }]
      }
      {
       type=MovieSceneFloatTrack
       display_name=Render Opacity
       sections [{
         type=MovieSceneFloatSection
         channels [None {
           type=MovieSceneScriptingFloatChannel
           keys [[0,0],[7.0703,1],[29.8347,1],[32.6017,0],[33.1597,1],[35.859,0],[36.283,1],[60.6537,1],[74.6537,0]]
          }]
        }]
      }
     ]
    }]}
 }]
graphs [EventGraph {nodes [
   n1 = EventShowDetail!() -> (OutputDelegate, then, ShowDetail:bool=~) {is=title K2Node_CustomEvent}
   > SetActiveWidgetIndex(
    self=Get CommonWidgetSwitcher_Detail() -> (CommonWidgetSwitcher_Detail:Common Widget Switcher Legacy Object Reference)
    Index=Select(Option 0:int=0, Option 1:int=1, Index=@n1.ShowDetail) -> (ReturnValue:int)
   )
   EventIntro() {is=title K2Node_CustomEvent}
   > PlayAnimation(InAnimation=Get IntroScale() -> (IntroScale:Widget Animation Object Reference))
  ]}]
variables []
widgets {
 info {inheritedWidgetCount=0, parentClass=&UserWidget, rootWidgetClass=&CommonWidgetSwitcherLegacy, widgetCount=8}
 widgets [
  CommonWidgetSwitcher_Detail {widgetClassPath=&CommonWidgetSwitcherLegacy, bIsVariable=true, widget=CommonWidgetSwitcher_Detail}
  Overlay_0 {
   widgetClassPath=&Overlay
   parent=CommonWidgetSwitcher_Detail
   widget=Overlay_0
   properties {renderTransform {angle=0, scale=<0.65 0.75>, shear=<0 0>, translation=<0 0>}}
   slot {class=WidgetSwitcherSlot, properties {horizontalAlignment=HAlign_Center, verticalAlignment=VAlign_Center}}
  }
  Image_63 {
   widgetClassPath=&Image
   bIsVariable=true
   parent=Overlay_0
   widget=Image_63
   properties {
    brush {
     bIsDynamicallyLoaded=false
     drawAs=Image
     imageType=NoImage
     mirroring=NoMirror
     resourceName=None
     resourceObject=/Game/UI/Login/SubGameSelect/Materials/M_UI_FocusBorderPulse
     tiling=NoTile
     imageSize=<32 32>
     margin=<ltrb 0 0 0 0>
     outlineSettings {
      bUseBrushTransparency=false
      roundingType=HalfHeightRadius
      width=0
      color {colorUseRule=UseColor_Specified, specifiedColor=<rgba 0 0 0 0>}
      cornerRadii=<0 0 0 0>
     }
     tintColor {colorUseRule=UseColor_Specified, specifiedColor=<rgba 1 1 1 1>}
     uVRegion {bIsValid=false, max=<0 0>, min=<0 0>}
    }
    colorAndOpacity=<rgba 0.4426 0 1 1>
    renderTransform {angle=-45, scale=<1 1>, shear=<0 0>, translation=<0 0>}
   }
   slot {class=OverlaySlot, properties {horizontalAlignment=HAlign_Fill, verticalAlignment=VAlign_Fill}}
  }
  Image {
   widgetClassPath=&Image
   bIsVariable=true
   parent=Overlay_0
   widget=Image
   properties {brush {
     bIsDynamicallyLoaded=false
     drawAs=Image
     imageType=NoImage
     mirroring=NoMirror
     resourceName=None
     resourceObject=/Game/Athena/HUD/LTM/Mash/Materials/M_UI_WorldObjectIndicator_Inst
     tiling=NoTile
     imageSize=<64 64>
     margin=<ltrb 0 0 0 0>
     outlineSettings {
      bUseBrushTransparency=false
      roundingType=HalfHeightRadius
      width=0
      color {colorUseRule=UseColor_Specified, specifiedColor=<rgba 0 0 0 0>}
      cornerRadii=<0 0 0 0>
     }
     tintColor {colorUseRule=UseColor_Specified, specifiedColor=<rgba 1 1 1 1>}
     uVRegion {bIsValid=false, max=<0 0>, min=<0 0>}
    }}
   slot {class=OverlaySlot, properties {horizontalAlignment=HAlign_Center, verticalAlignment=VAlign_Center}}
  }
  VerticalBox_18 {
   widgetClassPath=&VerticalBox
   parent=CommonWidgetSwitcher_Detail
   widget=VerticalBox_18
   slot {class=WidgetSwitcherSlot, properties {horizontalAlignment=HAlign_Center, verticalAlignment=VAlign_Center}}
  }
  CommonTextBlock_22 {
   widgetClassPath=&CommonTextBlock
   parent=VerticalBox_18
   widget=CommonTextBlock_22
   properties {
    text=Multiplier
    textTransformPolicy=ToUpper
    font {
     bForceMonospaced=false
     bMaterialIsStencil=false
     fontMaterial=null
     fontObject=&Burbank
     letterSpacing=0
     monospacedWidth=1
     size=24
     skewAmount=0
     typefaceFontName=Burbank-PlayerPower
     outlineSettings {
      bApplyOutlineToDropShadows=false
      bMiteredCorners=false
      bSeparateFillAlpha=false
      outlineMaterial=null
      outlineSize=1
      outlineColor=<rgba 0 0 0 1>
     }
    }
    shadowColorAndOpacity=<rgba 0.0052 0.0052 0.0056 1>
    shadowOffset=<4 4>
   }
   slot {class=VerticalBoxSlot, properties {horizontalAlignment=HAlign_Center, verticalAlignment=VAlign_Center, padding=<ltrb 0 0 0 -8>}}
  }
  CommonTextBlock {
   widgetClassPath=&CommonTextBlock
   parent=VerticalBox_18
   widget=CommonTextBlock
   properties {
    text=Bonus
    textTransformPolicy=ToUpper
    font {
     bForceMonospaced=false
     bMaterialIsStencil=false
     fontMaterial=null
     fontObject=&BurbankSmall
     letterSpacing=0
     monospacedWidth=1
     size=24
     skewAmount=0
     typefaceFontName=Black
     outlineSettings {
      bApplyOutlineToDropShadows=false
      bMiteredCorners=false
      bSeparateFillAlpha=false
      outlineMaterial=null
      outlineSize=1
      outlineColor=<rgba 0 0 0 1>
     }
    }
    renderTransform {angle=0, scale=<1 1>, shear=<-12 0>, translation=<0 0>}
    shadowColorAndOpacity=<rgba 0.0052 0.0052 0.0056 1>
    shadowOffset=<4 4>
   }
   slot {class=VerticalBoxSlot, properties {horizontalAlignment=HAlign_Center, verticalAlignment=VAlign_Center}}
  }
  Image_62 {
   widgetClassPath=&Image
   bIsVariable=true
   parent=VerticalBox_18
   widget=Image_62
   properties {brush {
     bIsDynamicallyLoaded=false
     drawAs=Image
     imageType=NoImage
     mirroring=NoMirror
     resourceName=None
     resourceObject=/Game/Athena/HUD/LTM/Mash/Materials/M_UI_MashArrow_Inst
     tiling=NoTile
     imageSize=<124.6503 112.1483>
     margin=<ltrb 0 0 0 0>
     outlineSettings {
      bUseBrushTransparency=false
      roundingType=HalfHeightRadius
      width=0
      color {colorUseRule=UseColor_Specified, specifiedColor=<rgba 0 0 0 0>}
      cornerRadii=<0 0 0 0>
     }
     tintColor {colorUseRule=UseColor_Specified, specifiedColor=<rgba 1 1 1 1>}
     uVRegion {bIsValid=false, max=<0 0>, min=<0 0>}
    }}
   slot {class=VerticalBoxSlot, properties {horizontalAlignment=HAlign_Center, verticalAlignment=VAlign_Bottom, padding=<ltrb 0 -80 0 0>}}
  }
 ]
}
