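# Translucent UI material for the Mash LTM HUD that draws a rounded-box border outline with an optional drop shadow.
# MF_UI_Box generates the outline from 9-slice UVs, with parameters for size, corner roundness and thickness, scaled by Opacity.
# A second box evaluation on offset UVs supplies a black shadow with its own opacity.
# Named reroutes carry both layers into lerps that composite color and alpha, with ShowShadow blending the shadow in.
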Material /Game/Athena/HUD/LTM/Mash/Materials/M_UI_Border_Rounded_Box
graph {
 properties {bAutomaticallySetUsageInEditor=false, blendMode=BLEND_Translucent, materialDomain=MD_UI, shadingModel=MSM_Unlit, usedShadingModels=Unlit}
 nodes [
  n1 = NamedRerouteUsage()
  LinearInterpolate(A=@n1, B=LinearInterpolate(A=NamedRerouteUsage(), B=@n1, Alpha=@n1), Alpha=ScalarParameter() {parameterName=ShowShadow}) {constAlpha=0}
  n2 = NamedRerouteUsage()
  n3 = NamedRerouteUsage()
  LinearInterpolate(A=@n2, B=LinearInterpolate(A=@n2, B=@n3, Alpha=@n3), Alpha=ScalarParameter() {parameterName=ShowShadow})
  n4 = TextureCoordinate()
  n5 = Multiply(A=AppendVector(A=ScalarParameter() {defaultValue=0.02, parameterName=ShadowPosX}, B=ScalarParameter() {defaultValue=0.02, parameterName=ShadowPosY})) {constB=-1}
  n6 = MaterialFunctionCall(
   UVs=MaterialFunctionCall!(UVs=Add(A=@n4, B=@n5) {constB=0.1}) {materialFunction=/FortUILibrary/MaterialFunctions/Basics/MF_UI_9SliceUVs}.UVs
   Size=ScalarParameter() {defaultValue=0.75, parameterName=Size}
   CornerRoundness=ScalarParameter() {defaultValue=0.2, parameterName=CornerRoundness}
   OutlineThickness=ScalarParameter() {defaultValue=0.05, parameterName=OutlineThickness}
  ) {materialFunction=&MF_UI_Box}
  NamedRerouteDeclaration(None=Multiply(A=@n6.Outline) {constB=0}) {name=ShadowColor, nodeColor=<rgba 0 0 0 1>}
  n7 = ScalarParameter() {defaultValue=0.75, parameterName=ShadowOpacity}
  n8 = ScalarParameter() {defaultValue=1, parameterName=Opacity, primitiveDataIndex=1, sliderMax=1}
  NamedRerouteDeclaration(None=Multiply(A=@n6.Outline, B=Multiply(A=@n7, B=@n8))) {name=ShadowOpacity, nodeColor=<rgba 0 0 0 1>}
  n9 = MaterialFunctionCall(UVs=TextureCoordinate()) {materialFunction=&MF_UI_9SliceUVs}
  n10 = ScalarParameter() {defaultValue=0.75, parameterName=Size}
  n11 = ScalarParameter() {defaultValue=0.2, parameterName=CornerRoundness}
  n12 = ScalarParameter() {defaultValue=0.05, parameterName=OutlineThickness}
  NamedRerouteDeclaration(None=Multiply(
    A=MaterialFunctionCall(UVs=@n9.UVs, Size=@n10, CornerRoundness=@n11, OutlineThickness=@n12) {materialFunction=&MF_UI_Box}.Outline
    B=ScalarParameter() {defaultValue=1, parameterName=Opacity, primitiveDataIndex=1, sliderMax=1}
   )) {name=Border, nodeColor=<rgba 0 0 0 1>}
 ]
}
