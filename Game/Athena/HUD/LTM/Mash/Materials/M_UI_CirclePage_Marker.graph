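# Translucent UI material that procedurally draws a circle marker, apparently a page indicator for the Mash LTM HUD, as a filled inner dot with an optional outer ring.
# One exponential radial gradient feeds SmoothStep thresholds, with the ring formed by subtracting two discs offset by OuterBorder.
# A static switch, ShowOuterCrcle, picks ring plus dot (combined by Max) or dot alone.
# EdgeSoftness feathers the edges, scaled by one minus InnerCircle for the dot.
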Material /Game/Athena/HUD/LTM/Mash/Materials/M_UI_CirclePage_Marker
graph {
 properties {blendMode=BLEND_Translucent, materialDomain=MD_UI, pixelDepthOffsetMode=PDOM_Legacy, usedShadingModels=Default Lit}
 nodes [
  Constant() {r=1}
  n1 = ScalarParameter() {defaultValue=0.4, parameterName=OuterCircle}
  n2 = MaterialFunctionCall() {materialFunction=&RadialGradientExponential}
  n3 = MaterialFunctionCall(Alpha=@n2.RadialGradientExponential, Min=@n1, Max=Add(A=@n1, B=ScalarParameter() {parameterName=EdgeSoftness})) {materialFunction=&SmoothStep}
  n4 = Add(A=@n1, B=ScalarParameter() {defaultValue=0.1619, parameterName=OuterBorder})
  n5 = MaterialFunctionCall(Alpha=@n2.RadialGradientExponential, Min=@n4, Max=Add(A=@n4, B=ScalarParameter() {parameterName=EdgeSoftness})) {materialFunction=&SmoothStep}
  n6 = ScalarParameter() {defaultValue=0.8, parameterName=InnerCircle}
  n7 = ScalarParameter() {parameterName=EdgeSoftness}
  n8 = OneMinus(None=@n6)
  n9 = MaterialFunctionCall(Alpha=@n2.RadialGradientExponential, Min=@n6, Max=Add(A=@n6, B=Multiply(A=@n7, B=@n8))) {materialFunction=&SmoothStep}
  StaticSwitchParameter(True=Max(A=Subtract(A=@n3.Result, B=@n5.Result), B=@n9.Result), False=@n9.Result) {defaultValue=true, parameterName=ShowOuterCrcle}
 ]
}
