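# Translucent UI material for a leaderboard star in the Mash LTM HUD, drawing an outlined star from a texture's alpha channel.
# A SmoothStep at OutlineEdge plus InnerEdge blends OutlineColor into the fill Color, with EdgeSoftness setting the transition width.
# A second SmoothStep at OutlineEdge forms the outer silhouette, combined by Max with the raw alpha.
# Default colors are a near-black outline and white fill.
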Material /Game/Athena/HUD/LTM/Mash/Materials/M_UI_LeaderboardStar
graph {
 properties {blendMode=BLEND_Translucent, materialDomain=MD_UI, pixelDepthOffsetMode=PDOM_Legacy, usedShadingModels=Default Lit}
 nodes [
  n1 = TextureSampleParameter2D() {parameterName=Texture, texture=&/RatingStarBig}
  n2 = ScalarParameter() {parameterName=OutlineEdge}
  n3 = Add(A=@n2, B=ScalarParameter() {parameterName=InnerEdge})
  n4 = ScalarParameter() {parameterName=EdgeSoftness}
  LinearInterpolate(
   A=VectorParameter() {parameterName=OutlineColor, defaultValue=<rgba 0.0052 0.0052 0.0056 1>}.RGB
   B=VectorParameter() {parameterName=Color, defaultValue=<rgba 1 1 1 1>}.RGB
   Alpha=MaterialFunctionCall(Alpha=@n1.A, Min=@n3, Max=Add(A=@n3, B=@n4)) {materialFunction=&SmoothStep}.Result
  )
  n5 = ScalarParameter() {parameterName=EdgeSoftness}
  Max(
   A=@n1.A
   B=MaterialFunctionCall(Alpha=@n1.A, Min=@n2, Max=Add(A=@n2, B=@n5)) {materialFunction=&SmoothStep}.Result
  )
 ]
}
