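# Translucent UI material for a locker page marker, drawing a grid of rounded boxes with optional drop shadows.
# Clamped UVs are multiplied into a 3x2 layout, scaled around a point and wrapped with Frac to tile MF_UI_Box.
# A duplicate chain with offset UVs builds the tinted shadow, blended in via ShowShadows.
# Parameters drive box size, spacing, roundness and clamping, though the upper clamp uses XMax for both axes.
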Material /Game/Athena/HUD/LTM/Mash/Materials/M_UI_LockerPage_Marker
graph {
 properties {bAutomaticallySetUsageInEditor=false, blendMode=BLEND_Translucent, materialDomain=MD_UI, usedShadingModels=Default Lit}
 nodes [
  n1 = NamedRerouteUsage()
  n2 = NamedRerouteUsage()
  LinearInterpolate(A=@n1, B=LinearInterpolate(A=@n1, B=@n2, Alpha=@n2), Alpha=ScalarParameter() {parameterName=ShowShadows, sliderMax=1})
  n3 = NamedRerouteUsage()
  LinearInterpolate(A=@n3, B=LinearInterpolate(A=NamedRerouteUsage(), B=@n3, Alpha=@n3), Alpha=ScalarParameter() {parameterName=ShowShadows, sliderMax=1})
  n4 = TextureCoordinate()
  n5 = VectorParameter() {parameterName=ClampingSquareUVs, channelNames {a=YMax, b=YMin, g=XMax, r=XMin}, defaultValue=<rgba 0.2 0.8 0.3 0.7>}
  n6 = ComponentMask(None=@n5) {b=true, r=true}
  n7 = AppendVector(A=@n5.XMax, B=@n5.XMax)
  n8 = MaterialFunctionCall!(
   Texture Scale=NamedRerouteUsage()
   UVs=Multiply(A=Clamp(None=@n4, Min=@n6, Max=@n7) {minDefault=1}, B=Constant2Vector() {g=2, r=3})
   Offset XY Texturesize Z=Constant3Vector() {constant=<rgba 1.5 1 1 0>}
  ) {materialFunction=/Engine/Functions/Engine_MaterialFunctions01/Texturing/ScaleUVsAroundPoint}
  NamedRerouteDeclaration(None=MaterialFunctionCall(UVs=Frac(None=@n8.UVs), Size=NamedRerouteUsage(), CornerRoundness=ScalarParameter() {defaultValue=0.1, parameterName=CornerRoundness}) {materialFunction=&MF_UI_Box}.Fill) {name=Boxes, nodeColor=<rgba 0 0 0 1>}
  n9 = VectorParameter() {parameterName=BoxSpacing, channelNames {a="", b="", g=Y, r=X}, defaultValue=<rgba 0.585 0.408 0 1>}
  n10 = ComponentMask(None=@n9) {g=true, r=true}
  NamedRerouteDeclaration(None=Multiply(
    A=ScalarParameter() {defaultValue=1, parameterName=BoxArrayScale}
    B=NamedRerouteDeclaration(None=@n10) {name=BoxSpacing, nodeColor=<rgba 0 0 0 1>}
   )) {name=BoxArrayScale, nodeColor=<rgba 0 0 0 1>}
  n11 = VectorParameter() {parameterName=BoxScale, channelNames {a="", b="", g=Y, r=X}, defaultValue=<rgba 0.5 0.3333 0 1>}
  NamedRerouteDeclaration(None=Divide(A=ComponentMask(None=@n11) {g=true, r=true}, B=NamedRerouteUsage())) {name=BoxSizeCorrection, nodeColor=<rgba 0 0 0 1>}
  n12 = NamedRerouteUsage()
  n13 = TextureCoordinate()
  n14 = Multiply(A=AppendVector(A=ScalarParameter() {defaultValue=0.02, parameterName=ShadowPosX}, B=ScalarParameter() {defaultValue=0.02, parameterName=ShadowPosY})) {constB=-1}
  n15 = VectorParameter() {parameterName=ClampingSquareUVs, channelNames {a=YMax, b=YMin, g=XMax, r=XMin}, defaultValue=<rgba 0.2 0.8 0.3 0.7>}
  n16 = Multiply(
   A=Clamp(None=Add(A=@n13, B=@n14) {constB=2}, Min=ComponentMask(None=@n15) {b=true, r=true}, Max=AppendVector(A=@n15.XMax, B=@n15.XMax)) {minDefault=1}
   B=Constant2Vector() {g=2, r=3}
  )
  n17 = Constant3Vector() {constant=<rgba 1.5 1 1 0>}
  n18 = MaterialFunctionCall(
   UVs=Frac(None=MaterialFunctionCall(Texture Scale=@n12, UVs=@n16, Offset XY Texturesize Z=@n17) {materialFunction=&ScaleUVsAroundPoint}.UVs)
   Size=NamedRerouteUsage()
   CornerRoundness=ScalarParameter() {defaultValue=0.1, parameterName=CornerRoundness}
  ) {materialFunction=&MF_UI_Box}
  NamedRerouteDeclaration(None=Multiply(A=@n18.Fill, B=VectorParameter() {parameterName=ShadowColor}.RGB)) {name=ShadowColor, nodeColor=<rgba 0 0 0 1>}
  NamedRerouteDeclaration(None=Multiply(A=@n18.Fill, B=ScalarParameter() {defaultValue=0.75, parameterName=ShadowOpacity})) {name=ShadowOpacity, nodeColor=<rgba 0 0 0 1>}
 ]
}
