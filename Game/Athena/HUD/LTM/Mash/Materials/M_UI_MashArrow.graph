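# UI masked material for the Mash LTM HUD that procedurally draws an arrow with an animated, offset shadow shape.
# Each arrow is a rotated, center-scaled UV square mask clipped vertically by a SmoothStep, requiring no textures.
# The fixed top arrow blends background and foreground colors.
# A second copy oscillates vertically on a one-second sine, has an inner arrow subtracted, and is combined with the top arrow through Max.
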Material /Game/Athena/HUD/LTM/Mash/Materials/M_UI_MashArrow
graph {
 properties {blendMode=BLEND_Masked, materialDomain=MD_UI, pixelDepthOffsetMode=PDOM_Legacy, usedShadingModels=Default Lit}
 nodes [
  Constant() {r=1}
  n1 = Add(A=TextureCoordinate(), B=AppendVector(A=Constant(), B=ScalarParameter() {defaultValue=0.2, parameterName=TopArrowOffset}))
  n2 = ScalarParameter() {defaultValue=0.125, group=Arrow, parameterName=ArrowAngle}
  n3 = ScalarParameter() {defaultValue=0.85, group=Arrow, parameterName=ArrowCutoff}
  n4 = ScalarParameter() {defaultValue=0.01, parameterName=EdgeSoftness}
  n5 = Multiply(
   A=MaterialFunctionCall(
    UVs=MaterialFunctionCall(UVs=@n1, "Rotation Angle (0-1)"=@n2) {materialFunction=&CustomRotator}.Rotated Values
    Texture Scale=ScalarParameter() {defaultValue=0.7, group=Arrow, parameterName=ArrowScale}
   ) {materialFunction=&ScaleUVsByCenter}.0-1 mask
   B=MaterialFunctionCall(Alpha=ComponentMask(None=@n1) {g=true}, Min=@n3, Max=Add(A=@n3, B=@n4)) {materialFunction=&SmoothStep}.Result
  )
  LinearInterpolate(
   A=VectorParameter() {parameterName=BackgroundColor, defaultValue=<rgba 0.402 0 1 1>}.RGB
   B=VectorParameter() {parameterName=ForegroundColor, defaultValue=<rgba 1 1 1 1>}.RGB
   Alpha=@n5
  )
  n6 = ScalarParameter() {parameterName=ShadowStart}
  n7 = Add(A=@n6, B=ScalarParameter() {defaultValue=0.2, parameterName=ShadowDistance})
  n8 = MaterialFunctionCall(Value=Time() {bOverride_Period=true, period=1}) {materialFunction=&LinearSine}
  n9 = Add(A=TextureCoordinate(), B=AppendVector(A=Constant(), B=LinearInterpolate(A=@n6, B=@n7, Alpha=@n8."Rounded Linear Sine ")))
  n10 = MaterialFunctionCall(
   UVs=MaterialFunctionCall(UVs=@n9, "Rotation Angle (0-1)"=ScalarParameter() {defaultValue=0.125, group=Arrow, parameterName=ArrowAngle}) {materialFunction=&CustomRotator}.Rotated Values
   Texture Scale=ScalarParameter() {defaultValue=0.7, group=Arrow, parameterName=ArrowScale}
  ) {materialFunction=&ScaleUVsByCenter}
  n11 = ScalarParameter() {defaultValue=0.85, group=Arrow, parameterName=ArrowCutoff}
  n12 = MaterialFunctionCall(Alpha=ComponentMask(None=@n9) {g=true}, Min=@n11, Max=Add(A=@n11, B=ScalarParameter() {defaultValue=0.01, parameterName=EdgeSoftness})) {materialFunction=&SmoothStep}
  n13 = MaterialFunctionCall(
   UVs=MaterialFunctionCall(UVs=@n9, "Rotation Angle (0-1)"=ScalarParameter() {defaultValue=0.125, group=Arrow, parameterName=ArrowAngle}) {materialFunction=&CustomRotator}.Rotated Values
   Texture Scale=ScalarParameter() {defaultValue=0.7, group=Arrow, parameterName=ArrowInnerScale}
  ) {materialFunction=&ScaleUVsByCenter}
  n14 = ScalarParameter() {defaultValue=0.85, group=Arrow, parameterName=ArrowInnerCutoff}
  n15 = MaterialFunctionCall(Alpha=ComponentMask(None=@n9) {g=true}, Min=@n14, Max=Add(A=@n14, B=ScalarParameter() {defaultValue=0.01, parameterName=EdgeSoftness})) {materialFunction=&SmoothStep}
  Max(A=@n5, B=Subtract(A=Multiply(A=@n10.0-1 mask, B=@n12.Result), B=Multiply(A=@n13.0-1 mask, B=@n15.Result)))
 ]
}
