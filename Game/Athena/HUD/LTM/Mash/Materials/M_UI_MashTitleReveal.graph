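# Translucent UI material that reveals a Mash LTM title through a noise-broken radial mask.
# StaticUVs blends screen-aligned to widget UVs, feeding center-scaled ice noise and an exponential radial gradient.
# An inverted SmoothStep thresholds noise against that gradient with EdgeSoftness; UseImage swaps white for an image, capping the mask by its alpha.
# PatternOffset's sum is unused, and Position uses only R for both axes.
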Material /Game/Athena/HUD/LTM/Mash/Materials/M_UI_MashTitleReveal
graph {
 properties {blendMode=BLEND_Translucent, materialDomain=MD_UI, pixelDepthOffsetMode=PDOM_Legacy, usedShadingModels=Default Lit}
 nodes [
  n1 = LinearInterpolate(
   A=MaterialFunctionCall(Texture Dimensions=MaterialFunctionCall() {materialFunction=&GetUserInterfaceUV}.Normalized UV) {materialFunction=&ScreenAlignedUVs}."X 100%, Y 100%"
   B=TextureCoordinate()
   Alpha=ScalarParameter() {parameterName=StaticUVs}
  )
  n2 = MaterialFunctionCall(UVs=@n1, Texture Scale=ScalarParameter() {defaultValue=20, group=Pattern, parameterName=PatternAmount}) {materialFunction=&ScaleUVsByCenter}
  n3 = VectorParameter() {group=Pattern, parameterName=PatternOffset, defaultValue=<rgba 0 0 0 1>}
  Add(A=@n2.UVs, B=AppendVector(A=@n3.R, B=@n3.R))
  n4 = TextureSampleParameter2D() {parameterName=Image}
  n5 = TextureSampleParameter2D(UVs=@n2.UVs) {group=Pattern, parameterName=Texture, texture=/Game/Global/Textures/Default/Noise/T_UI_Ice_Noise}
  n6 = ScalarParameter() {group=Reveal, parameterName=EdgeSoftness}
  n7 = VectorParameter() {parameterName=Position, defaultValue=<rgba 0.5 0.5 0 1>}
  n8 = MaterialFunctionCall(
   UVs=@n1
   CenterPosition=AppendVector(A=@n7.R, B=@n7.R)
   Radius=ScalarParameter() {group=Reveal, parameterName=Radius}
   Density=ScalarParameter() {group=Reveal, parameterName=Density}
  ) {materialFunction=&RadialGradientExponential}
  n9 = OneMinus(None=MaterialFunctionCall(
    Alpha=Clamp(None=@n5.R) {maxDefault=0.99, minDefault=0.01}
    Min=@n8.RadialGradientExponential
    Max=Add(A=@n8.RadialGradientExponential, B=@n6)
   ) {materialFunction=&SmoothStep}.Result)
  StaticSwitchParameter(True=Min(A=@n4.A, B=@n9), False=@n9) {parameterName=UseImage}
  StaticSwitchParameter(True=@n4.RGB, False=Constant() {r=1}) {parameterName=UseImage}
 ]
}
