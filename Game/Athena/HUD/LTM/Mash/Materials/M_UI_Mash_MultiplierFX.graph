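# Additive UI material for the Mash LTM HUD, producing an animated noise effect for the multiplier FX.
# Cloud noise pans vertically on center-scaled UVs and is sampled both directly and through a radial conversion of those coordinates.
# The maximum of both samples is raised to a parameterized power, negative by default so dark noise turns bright.
# An exponential radial gradient masks the result, which then multiplies the MultiplierFX texture color.
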Material /Game/Athena/HUD/LTM/Mash/Materials/M_UI_Mash_MultiplierFX
graph {
 properties {blendMode=BLEND_Additive, materialDomain=MD_UI, pixelDepthOffsetMode=PDOM_Legacy, usedShadingModels=Default Lit}
 nodes [
  n1 = ScalarParameter() {defaultValue=0.5, parameterName=Radius}
  n2 = ScalarParameter() {defaultValue=2.3, parameterName=Density}
  n3 = Constant()
  n4 = ScalarParameter() {defaultValue=0.3619, parameterName=PanSpeed}
  n5 = Panner(
   Coordinate=MaterialFunctionCall(Texture Scale=ScalarParameter() {defaultValue=1, parameterName=PatternAmount}) {materialFunction=&ScaleUVsByCenter}.UVs
   Speed=AppendVector(A=AppendVector(A=@n3, B=@n4), B=@n3)
  ) {speedY=0.3619}
  n6 = Max(
   A=TextureSample!(UVs=@n5, Tex, Bias, Apply View MipBias) {constMipValue=3, mipValueMode=TMVM_MipBias, texture=&AtmosphericCloudNoise01}.R
   B=TextureSample(UVs=MaterialFunctionCall(Vector or UVs=@n5) {materialFunction=&VectorToRadialValue}.Radial Coordinates) {constMipValue=3, mipValueMode=TMVM_MipBias, texture=&AtmosphericCloudNoise01}.R
  )
  n7 = ScalarParameter() {defaultValue=-2, parameterName=Power}
  Multiply(
   A=TextureSample!(UVs, Tex, Apply View MipBias) {texture=&/T_UI_Mash_MultiplierFX}.RGB
   B=Multiply(
    A=MaterialFunctionCall(Radius=@n1, Density=@n2) {materialFunction=&RadialGradientExponential}.RadialGradientExponential
    B=Power(Base=@n6, Exp=@n7)
   )
  )
 ]
}
