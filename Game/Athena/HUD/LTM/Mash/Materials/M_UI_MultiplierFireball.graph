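# Translucent UI material that draws an animated, stylized fireball, apparently for the multiplier display in the Mash LTM HUD.
# Two panning noise textures form the flame intensity, shaped by radial gradients, a vertical ramp and an optional distance-field mask.
# That intensity drives a two-color blend, with ColorSwitch fading to a vertical gradient plus scrolling embers.
# Smoothstep thresholds and a bump-offset-distorted mask texture cut the silhouette, with a sine-pulsed edge softness.
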Material /Game/Athena/HUD/LTM/Mash/Materials/M_UI_MultiplierFireball
graph {
 properties {blendMode=BLEND_Translucent, materialDomain=MD_UI, numCustomizedUVs=1, pixelDepthOffsetMode=PDOM_Legacy, usedShadingModels=Default Lit}
 nodes [
  n1 = VectorParameter() {parameterName=Color2, defaultValue=<rgba 0.3281 0.0154 0 1>}
  n2 = VectorParameter() {parameterName=Color1, defaultValue=<rgba 1 0.8698 0.141 1>}
  n3 = ScalarParameter() {defaultValue=1, parameterName=TopAdjust}
  n4 = ScalarParameter() {parameterName=ActivateAdjust}
  n5 = Add(A=TextureCoordinate(), B=AppendVector(A=Constant(), B=Add(A=@n3, B=@n4)))
  n6 = TextureCoordinate()
  n7 = Multiply(A=TextureCoordinate(), B=ScalarParameter() {defaultValue=1, parameterName=PatternScale})
  n8 = MaterialFunctionCall(Float2=MaterialFunctionCall() {materialFunction=&GetUserInterfaceUV}.Pixel Size) {materialFunction=&BreakOutFloat2Components}
  n9 = VectorParameter() {parameterName=InterfaceScale, defaultValue=<rgba 1 1 0 1>}
  n10 = Multiply(A=AppendVector(A=Divide(A=@n8.R, B=@n8.R), B=Constant() {r=1}), B=AppendVector(A=@n9.R, B=@n9.R)) {desc=Scale X to fit}
  n11 = Multiply(A=@n7, B=@n10)
  n12 = ScalarParameter() {parameterName=ScreenAlignedUVs}
  n13 = VectorParameter() {parameterName=Speed2, defaultValue=<rgba 0 0.5 0 1>}
  n14 = Add(A=Time(), B=ScalarParameter() {parameterName=TimeOffset})
  n15 = TextureSampleParameter2D(UVs=Panner(Coordinate=LinearInterpolate(A=@n6, B=@n11, Alpha=@n12), Time=@n14, Speed=AppendVector(A=@n13.R, B=@n13.R)) {speedY=0.5}) {parameterName=FireTexture2, texture=/Game/Effects/Fort_Effects/Textures/Fire/T_FireNoise_02}
  n16 = Power(Base=@n15.R, Exp=ScalarParameter() {defaultValue=-0.75, parameterName=FlameExp}) {constExponent=-0.75}
  n17 = Multiply(A=@n7)
  n18 = ScalarParameter() {defaultValue=2, parameterName=FlameDistort}
  n19 = VectorParameter() {parameterName=Speed1, defaultValue=<rgba 0 0.2 0 1>}
  n20 = TextureSampleParameter2D(UVs=Panner(
    Coordinate=MaterialFunctionCall!(HeightTexture=@n15.R, TextureCoordinates=@n17, ReferencePlane, HeightRatioInput=@n18, CameraVector) {materialFunction=/Engine/Functions/Engine_MaterialFunctions02/Utility/BumpOffset_Advanced.BumpOffset_advanced}.Undeformed UVs
    Time=@n14
    Speed=AppendVector(A=@n19.R, B=@n19.R)
   ) {speedY=0.2}) {parameterName=FireTexture1, texture=/Game/Global/Textures/Default/Noise/T_FireNoise_01}
  n21 = Constant2Vector() {g=1, r=0.5}
  n22 = MaterialFunctionCall(Value=@n14) {materialFunction=&LinearSine}
  n23 = LinearInterpolate(A=Constant() {r=1}, B=Constant() {r=2}, Alpha=@n22.Linear Sine)
  n24 = Constant() {r=0.7164}
  n25 = Power(
   Base=@n20.R
   Exp=Multiply(A=MaterialFunctionCall(UVs=@n17, CenterPosition=@n21, Radius=@n23, Density=@n24) {materialFunction=&RadialGradientExponential}.RadialGradientExponential) {constB=-1}
  ) {constExponent=5}
  n26 = VectorParameter() {group=TopGradient, parameterName=TopGradientScale, defaultValue=<rgba 0 0 0 1>}
  n27 = VectorParameter() {group=TopGradient, parameterName=TopGradientPosition, defaultValue=<rgba 0.5 0 0 0>}
  n28 = MaterialFunctionCall(
   UVs=MaterialFunctionCall(
    UVs=MaterialFunctionCall() {materialFunction=&GetUserInterfaceUV}.Normalized UV
    Texture Scale=AppendVector(A=@n26.R, B=@n26.R)
   ) {materialFunction=&ScaleUVsByCenter}.UVs
   CenterPosition=AppendVector(A=@n27.R, B=@n27.R)
   Radius=@n27.B
   Density=@n27.B
  ) {materialFunction=&RadialGradientExponential}
  n29 = Constant2Vector() {g=2, r=3}
  n30 = VectorParameter() {parameterName=FlameHeight, defaultValue=<rgba 0 0 0 1>}
  n31 = VectorParameter() {parameterName=FlameCenterPosition, defaultValue=<rgba 0.5 1 0.8043 2.3384>}
  n32 = MaterialFunctionCall(
   UVs=Add(
    A=MaterialFunctionCall(Texture Scale=@n29) {materialFunction=&ScaleUVsByCenter}.UVs
    B=AppendVector(A=@n30.R, B=@n30.R)
   )
   CenterPosition=@n31.RGB
   Radius=@n31.RGB
   Density=@n31.RGB
   Invert Density=StaticBool() {value=true}
  ) {materialFunction=&RadialGradientExponential}
  n33 = ScalarParameter() {defaultValue=1, group=DFMask, parameterName=DFScale}
  n34 = ScalarParameter() {defaultValue=1, parameterName=DFActivate}
  n35 = TextureSampleParameter2D(UVs=MaterialFunctionCall!(
    Texture Scale=Multiply(A=@n33, B=@n34)
    UVs=MaterialFunctionCall() {materialFunction=&GetUserInterfaceUV}.Normalized UV
    Offset XY Texturesize Z=VectorParameter() {group=DFMask, parameterName=DFScalePivot, defaultValue=<rgba 0.5 1 1 0>}.RGB
   ) {materialFunction=/Engine/Functions/Engine_MaterialFunctions01/Texturing/ScaleUVsAroundPoint}.UVs) {
   group=DFMask
   parameterName=DFMaskTexture
   samplerSource=SSM_Clamp_WorldGroupSettings
   texture=/Game/Athena/HUD/Materials/Score/T_UI_Square_DF
  }
  n36 = ScalarParameter() {group=DFMask, parameterName=DFMask}
  n37 = Multiply(
   A=Multiply(A=ComponentMask(None=@n5) {g=true}, B=LinearInterpolate(B=@n16, Alpha=@n25))
   B=Multiply(A=OneMinus(None=@n28.RadialGradientExponential), B=LinearInterpolate(A=@n32.RadialGradientExponential, B=@n35.R, Alpha=@n36))
  )
  n38 = Power(Base=@n37, Exp=ScalarParameter() {defaultValue=2, parameterName=Firepower})
  n39 = MaterialFunctionCall(
   Alpha=@n37
   Min=ScalarParameter() {defaultValue=0.1, parameterName=MaskMin}
   Max=ScalarParameter() {defaultValue=0.76, parameterName=MaskMax}
  ) {materialFunction=&SmoothStep}
  n40 = Power(Base=@n39.Result)
  n41 = Clamp(None=@n40)
  n42 = ComponentMask(None=TextureCoordinate()) {g=true}
  n43 = VectorParameter() {parameterName=ColorSwitch}
  n44 = Saturate(None=Add(A=Power(Base=@n42, Exp=@n43.G), B=@n43.B))
  n45 = TextureCoordinate()
  n46 = VectorParameter() {group=Embers, parameterName=EmberSpeed, defaultValue=<rgba 0 0.5 0 1>}
  n47 = TextureSampleParameter2D(UVs=Panner(Coordinate=Multiply(A=@n45, B=@n10), Speed=AppendVector(A=@n46.R, B=@n46.R))) {group=Embers, parameterName=EmberTexture, texture=/Game/Global/Textures/Default/PointSplatter/T_UI_StarBurst_Scatter}
  n48 = Multiply(
   A=LinearInterpolate(
    A=VectorParameter() {group=Embers, parameterName=EmberColor, defaultValue=<rgba 0.3281 0.0154 0 1>}.RGB
    B=VectorParameter() {group=Embers, parameterName=EmberColor2, defaultValue=<rgba 1 0.8698 0.141 1>}.RGB
    Alpha=@n38
   )
   B=ScalarParameter() {defaultValue=1, group=Embers, parameterName=EmberBoost}
  )
  n49 = OneMinus(None=@n40)
  n50 = Multiply(
   A=Multiply(A=Multiply(A=@n47.R, B=@n48), B=@n37)
   B=Multiply(A=ConstantBiasScale(None=@n49) {bias=0, scale=1}, B=ScalarParameter() {defaultValue=1, group=Embers, parameterName=EmberBoost2})
  )
  LinearInterpolate(
   A=LinearInterpolate(A=@n1.RGB, B=@n2.RGB, Alpha=Multiply(A=@n38, B=@n41))
   B=Add(A=LinearInterpolate(A=@n1.RGB, B=@n2.RGB, Alpha=@n44), B=@n50)
   Alpha=@n43.R
  )
  MaterialFunctionCall(HeightTexture=@n20.R, TextureCoordinates=@n17, HeightRatioInput=@n18) {materialFunction=&BumpOffset_advanced}
  n51 = VectorParameter() {group=OuterShape, parameterName=OuterShapePosition, defaultValue=<rgba 0.5 0.5 0 1>}
  n52 = OneMinus(None=MaterialFunctionCall(
    CenterPosition=AppendVector(A=@n51.R, B=@n51.R)
    Radius=ScalarParameter() {defaultValue=0.5, group=OuterShape, parameterName=OuterShapeRadius}
    Density=ScalarParameter() {defaultValue=2.3, group=OuterShape, parameterName=OuterShapeDensity}
   ) {materialFunction=&RadialGradientExponential}.RadialGradientExponential)
  MaterialFunctionCall(Alpha=@n39.Result, Min=@n52, Max=Add(A=@n52) {constB=0.01}) {materialFunction=&SmoothStep}
  n53 = LinearInterpolate(Alpha=@n37)
  n54 = MaterialFunctionCall() {materialFunction=&GetUserInterfaceUV}
  n55 = ScalarParameter() {defaultValue=0.5, group=FlameMask, parameterName=OutlineBumpAngle}
  n56 = ScalarParameter() {defaultValue=0.1, group=FlameMask, parameterName=OutlineDistort}
  n57 = VectorParameter() {group=FlameMask, parameterName=OutlineDistortVector, defaultValue=<rgba 0 0 0 1>}
  n58 = VectorParameter() {group=FlameMask, parameterName=MaskScale, defaultValue=<rgba 1 1 0 1>}
  n59 = TextureSampleParameter2D(UVs=MaterialFunctionCall(
    UVs=MaterialFunctionCall(HeightTexture=@n53, TextureCoordinates=@n54.9-Slice UV, ReferencePlane=@n55, HeightRatioInput=@n56, CameraVector=@n57.RGB) {materialFunction=&BumpOffset_advanced}.UVs
    Texture Scale=AppendVector(A=@n58.R, B=@n58.R)
   ) {materialFunction=&ScaleUVsByCenter}.UVs) {
   group=FlameMask
   parameterName=MaskTexture
   samplerSource=SSM_Clamp_WorldGroupSettings
   texture=&/T_UI_Mash_FlameEffect_Mask
  }
  n60 = OneMinus(None=@n59.R)
  n61 = ScalarParameter() {defaultValue=0.001, group=FlameMask, parameterName=OutlineSmooth}
  n62 = Add(A=@n60, B=LinearInterpolate(A=@n61, B=Multiply(A=@n61) {constB=0.1}, Alpha=@n22.Linear Sine)) {constB=0.01}
  Max(
   A=Multiply(
    A=@n41
    B=MaterialFunctionCall(Alpha=@n39.Result, Min=@n60, Max=@n62) {materialFunction=&SmoothStep}.Result
   )
   B=Saturate(None=@n50)
  )
  TextureCoordinate()
  TextureCoordinate()
 ]
}
