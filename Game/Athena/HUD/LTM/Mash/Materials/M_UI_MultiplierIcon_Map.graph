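# Translucent UI material for the Mash LTM multiplier icon, by its name for the map, drawing an outlined icon with a repeating pulse ring.
# One icon texture is sampled at several center-scaled sizes to build the icon, its border and, by subtracting alphas, the ring.
# Looping time with a power falloff drives ring size, thickness and fade.
# PulseColor tints outside the bordered icon, whose alpha is maxed with the ring.
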Material /Game/Athena/HUD/LTM/Mash/Materials/M_UI_MultiplierIcon_Map
graph {
 properties {blendMode=BLEND_Translucent, materialDomain=MD_UI, pixelDepthOffsetMode=PDOM_Legacy, usedShadingModels=Default Lit}
 nodes [
  n1 = ScalarParameter() {defaultValue=0.5, parameterName=IconScale}
  n2 = TextureSample(UVs=MaterialFunctionCall(Texture Scale=@n1) {materialFunction=&ScaleUVsByCenter}.UVs) {samplerSource=SSM_Clamp_WorldGroupSettings, texture=&/T_UI_Mash_MultiplierIcon}
  n3 = ScalarParameter() {defaultValue=1.2, parameterName=IconBorderSize}
  n4 = TextureSample(UVs=MaterialFunctionCall(Texture Scale=Multiply(A=@n1, B=@n3)) {materialFunction=&ScaleUVsByCenter}.UVs) {samplerSource=SSM_Clamp_WorldGroupSettings, texture=&T_UI_Mash_MultiplierIcon}
  LinearInterpolate(
   A=VectorParameter() {parameterName=PulseColor, defaultValue=<rgba 1 1 1 1>}.RGB
   B=LinearInterpolate(A=VectorParameter() {parameterName=IconOutlineColor, defaultValue=<rgba 1 0.7865 0.8683 1>}.RGB, B=@n2.RGB, Alpha=@n2.RGB)
   Alpha=@n4.A
  )
  n5 = Fmod(A=Multiply(A=Time(), B=ScalarParameter() {defaultValue=1, parameterName=PulseSpeed}), B=Constant() {r=1})
  MaterialFunctionCall(Value=@n5) {materialFunction=&LinearSine}
  n6 = Power(Base=@n5, Exp=ScalarParameter() {defaultValue=0.5, parameterName=PulseFalloff}) {constExponent=0.5}
  n7 = LinearInterpolate(Alpha=@n6) {constA=2, constB=0.5}
  n8 = LinearInterpolate(
   A=ScalarParameter() {defaultValue=0.5, parameterName=OutlineSizeMin}
   B=ScalarParameter() {defaultValue=0.9, parameterName=OutlineSize}
   Alpha=@n6
  )
  n9 = TextureSample(UVs=MaterialFunctionCall(Texture Scale=@n8) {materialFunction=&ScaleUVsByCenter}.UVs) {samplerSource=SSM_Clamp_WorldGroupSettings, texture=&T_UI_Mash_MultiplierIcon}
  n10 = LinearInterpolate(
   A=ScalarParameter() {defaultValue=0.8, parameterName=OutlineBorder}
   B=ScalarParameter() {defaultValue=0.9, parameterName=OutlineBorderMin}
   Alpha=@n6
  )
  n11 = TextureSample(UVs=MaterialFunctionCall(Texture Scale=Multiply(A=@n8, B=@n10)) {materialFunction=&ScaleUVsByCenter}.UVs) {samplerSource=SSM_Clamp_WorldGroupSettings, texture=&T_UI_Mash_MultiplierIcon}
  Max(A=Multiply(A=Clamp(None=@n7), B=Subtract(A=@n9.A, B=@n11.A)), B=@n4.A)
 ]
}
