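# Translucent UI material for the Mash LTM HUD's world object indicator, forming a hollow square frame shape.
# Texture UVs pass through CustomRotator, driven by a Rotation parameter, then feed two ScaleUVsByCenter calls.
# The InnerScale mask is subtracted from the Scale mask to cut out the center.
# Each scale vector uses only its R channel for both axes, keeping the scaling uniform.
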Material /Game/Athena/HUD/LTM/Mash/Materials/M_UI_WorldObjectIndicator
graph {
 properties {blendMode=BLEND_Translucent, materialDomain=MD_UI, pixelDepthOffsetMode=PDOM_Legacy, usedShadingModels=Default Lit}
 nodes [
  Constant() {r=1}
  n1 = MaterialFunctionCall(UVs=TextureCoordinate(), "Rotation Angle (0-1)"=ScalarParameter() {defaultValue=0.25, parameterName=Rotation}) {materialFunction=&CustomRotator}
  n2 = VectorParameter() {parameterName=Scale, defaultValue=<rgba 1 1 0 1>}
  n3 = VectorParameter() {parameterName=InnerScale, defaultValue=<rgba 0.5 0.5 0 1>}
  Subtract(
   A=MaterialFunctionCall(UVs=@n1.Rotated Values, Texture Scale=AppendVector(A=@n2.R, B=@n2.R)) {materialFunction=&ScaleUVsByCenter}.0-1 mask
   B=MaterialFunctionCall(UVs=@n1.Rotated Values, Texture Scale=AppendVector(A=@n3.R, B=@n3.R)) {materialFunction=&ScaleUVsByCenter}.0-1 mask
  )
 ]
}
