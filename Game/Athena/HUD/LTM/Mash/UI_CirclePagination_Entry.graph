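# Circle pagination dot widget for the Mash LTM HUD, showing whether its page is selected.
# Construct applies the unselected marker material to the single circle image.
# The EventSelected custom event swaps between selected and unselected materials based on its bool, then plays the Select animation when selected.
# That short animation loops indefinitely, animating the material's OuterCircle and OuterBorder parameters; nothing in the graph stops it.
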WidgetBlueprint /Game/Athena/HUD/LTM/Mash/UI_CirclePagination_Entry
animations [Select {
  binding_count=1
  length_seconds=0.3712
  object=&self:Select
  sequence {bindings [Image_CircleEntry {
     id=#1
     possessed_class=Image
     tracks [{
       type=MovieSceneWidgetMaterialTrack
       sections [{
         type=MovieSceneParameterSection
         end_frame=7
         start_frame=0
         channels [
          OuterCircle {type=MovieSceneScriptingFloatChannel, keys [[0,0.0339],[4.1443,0.7297],[5.9397,0.8939]]}
          OuterBorder {type=MovieSceneScriptingFloatChannel, keys [[0,0.1281],[4.549,0.0332],[5.9397,0.0381]]}
         ]
        }]
      }]
    }]}
 }]
graphs [EventGraph {nodes [
   n1 = EventSelected!() -> (OutputDelegate, then, isSelected:bool=~) {is=title K2Node_CustomEvent}
   > SetBrushFromMaterial(
    self=Get Image_CircleEntry() -> (Image_CircleEntry:Image Object Reference)
    Material=Select(
     Option 0:Material Interface Object Reference=/Game/Athena/HUD/LTM/Mash/Materials/M_UI_CirclePage_Marker_Selected_Unselected
     Option 1:Material Interface Object Reference=/Game/Athena/HUD/LTM/Mash/Materials/M_UI_CirclePage_Marker_Selected
     Index=@n1.isSelected
    ) -> (ReturnValue:Material Interface Object Reference)
   )
   > Branch(Condition=@n1.isSelected)
   > PlayAnimation(InAnimation=Get Select() -> (Select:Widget Animation Object Reference), NumLoopsToPlay=0)
   Event Construct()
   > SetBrushFromMaterial(self=Get Image_CircleEntry(), Material=&M_UI_CirclePage_Marker_Selected_Unselected)
  ]}]
variables []
widgets {
 info {inheritedWidgetCount=0, parentClass=&UserWidget, rootWidgetClass=&Overlay, widgetCount=3}
 widgets [
  Overlay_0 {widgetClassPath=&Overlay, widget=Overlay_0, properties {visibility=HitTestInvisible}}
  CommonBorder_0 {
   widgetClassPath=&CommonBorder
   parent=Overlay_0
   widget=CommonBorder_0
   properties {padding=<ltrb -14 -14 -14 -14>}
   slot {class=OverlaySlot}
  }
  Image_CircleEntry {
   widgetClassPath=&Image
   bIsVariable=true
   parent=CommonBorder_0
   widget=Image_CircleEntry
   properties {brush {
     bIsDynamicallyLoaded=false
     drawAs=Image
     imageType=NoImage
     mirroring=NoMirror
     resourceName=None
     resourceObject=&M_UI_CirclePage_Marker_Selected
     tiling=NoTile
     imageSize=<64 64>
     margin=<ltrb 0 0 0 0>
     outlineSettings {
      bUseBrushTransparency=false
      roundingType=HalfHeightRadius
      width=0
      color {colorUseRule=UseColor_Specified, specifiedColor=<rgba 0 0 0 0>}
      cornerRadii=<0 0 0 0>
     }
     tintColor {colorUseRule=UseColor_Specified, specifiedColor=<rgba 1 1 1 1>}
     uVRegion {bIsValid=false, max=<0 0>, min=<0 0>}
    }}
   slot {class=BorderSlot, properties {padding=<ltrb -14 -14 -14 -14>}}
  }
 ]
}
